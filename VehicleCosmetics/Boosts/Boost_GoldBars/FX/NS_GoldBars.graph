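# Niagara system for the Gold Bars vehicle boost cosmetic, spraying gold bars, occasional trophies, lens flares and embers behind the car.
# All five emitters spawn near System.FinalPosition, with skeletal-mesh socket sampling disabled.
# Bars and trophies tumble under gravity, briefly inherit source movement and align on collision.
# User.bIsFrontendPreview switches between two collision setups, one using an offset plane for previews, while User.bIsPainted recolors only the embers.
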NiagaraSystem /VehicleCosmetics/Boosts/Boost_GoldBars/FX/NS_GoldBars
dependencies {
 usedDataInterfaces [
  &NiagaraDataInterfaceCurve
  &NiagaraDataInterfaceArrayFloat2
  &NiagaraDataInterfaceVectorCurve
  &NiagaraDataInterfaceSpriteRendererInfo
  &NiagaraDataInterfaceSkeletalMesh
 ]
 usedDynamicInputs [/VehicleCosmetics/FX/Niagara/Modules/MergeEmitters/VC_FloatBySpawnGroup, &SelectVector2DFromArray]
 usedModules [
  &self:SystemSpawnScript.SystemScriptSource.SystemScriptGraph.NiagaraNodeAssignment_0.SetVariables_#1
  &SystemState
 ]
}
emitters {
 Goldbar {
  emitterUpdateScript [
   &EmitterState()
   &VC_FilteredSocketCount()
   &SpawnRate(SpawnRate=30, Use Spawn Probability=false)
  ]
  particleSpawnScript [
   &VC_PaintedVariantsOverride() {enabled=false}
   &InitializeParticle(
    Color Mode=::NewEnumerator0
    Lifetime=&RandomRangeFloat(Minimum=2, Maximum=1)
    Mesh Scale Mode=::NewEnumerator3
    Position Offset=&AddVectorToPosition(Position=$System.FinalPosition, Vector=<-30 0 0>)
    Position Offset Coordinate Space=::Local
    Sprite Rotation Mode=::NewEnumerator2
    UsePositionOffset=true
    Write Material Random=true
   )
   &PartitionParticles(Number Of Partitions=$Emitter.PortCount) {enabled=false}
   &SkeletalMeshLocation(
    Bone / Socket Index=$Particles.PartitionParticles.Partition
    Bone / Socket Sampling Mode=::NewEnumerator5
    MaskSpawnGroup=false
    Mesh Sampling Type=::NewEnumerator6
    Sampled Position Offset=&AddVectorToPosition(Position=<-20 0 0>, Vector=Linked() {linkedVariable {name=User.DTCarOffset, type=NiagaraPosition}})
    Skeletal Mesh=$System.SkeletalMesh
   ) {enabled=false}
   &AddVelocity(Velocity=&RandomRangeVector(Minimum=<-600 -120 300>, Maximum=<-1000 120 50>))
   &Color(Color=<rgba 0.4635 0.2961 0.0469 1>)
   &InitialMeshOrientation(
    Mesh Orientation Mode=::NewEnumerator7
    Rotation=&RandomRangeVector(Minimum=<-1 -1 -1>, Maximum=<1 1 1>)
    Use Rotation Vector=true
   )
  ]
  particleUpdateScript [
   &ParticleState()
   &GravityForce(Gravity=&RandomRangeVector(Minimum=<0 0 -2000>, Maximum=<0 0 -1200>))
   &InheritSourceMovement(
    Applied Position Scale=&VectorFromCurve(VectorCurve=DataInterface() {propertyValues {XCurve {keys [<at 0.1 0.7>,<at 0.5 0>]}, YCurve {keys [<at 0.1 0.7>,<at 0.5 0>]}, ZCurve {keys [<at 0.1 0.7>,<at 0.5 0>]}}})
    ApplyForce=false
    ApplyPosition=true
    ApplyPositionFalloff=false
   )
   &Collision(
    Advanced Aging Rate=5
    Age Colliding Particles=true
    Collision Enabled=HlslExpression() {hlslExpression="User.bIsFrontendPreview == true ? false : true"}
    Control Roll On Collision=true
    Friction=1.5
    Friction During a Bounce:NiagaraFloat=&RandomRangeFloat(Minimum=2, Maximum=1.2)
    KillOnCollision=false
    Module.Control Rest State Via Surface Normal=false
    Module.EnableMaxCollisionCount=false
    Module.Randomize Collision Normal Vector=false
    Radius Scale for Rotation=0
    Restitution=&RandomRangeFloat(Minimum=0, Maximum=0.1)
    ScaleRateOfRotation=true
   )
   &Collision(
    Advanced Aging Rate=0.3
    Age Colliding Particles=true
    Analytical Collision Plane Position 1:Vector3f=<0 0 40>
    CPU Collision Type=::NewEnumerator3
    CPU Friction Blending:ENiagaraFrictionMergeType=::NewEnumerator1
    Collision Enabled=HlslExpression() {hlslExpression="User.bIsFrontendPreview == true ? true : false"}
    Control Roll On Collision=true
    Enable Rest State=false
    Friction=1
    Friction During a Bounce:NiagaraFloat=&RandomRangeFloat(Minimum=2, Maximum=1)
    KillOnCollision=false
    Module.EnableMaxCollisionCount=false
    Module.Randomize Collision Normal Vector=true
    Radius Calculation Type=::NewEnumerator1
    Radius Scale for Rotation=20
    Restitution=&RandomRangeFloat(Minimum=0.3, Maximum=0.1)
    Restitution Coefficient Blending:ENiagaraRestitutionMergeType=::NewEnumerator2
    ScaleRateOfRotation=true
   ) {module=Collision001}
   /Niagara/Modules/Collision/AlignParticlesWithCollisionPlane(Allow Alignment on Y=false, Force Alignment=true, Rotational Dampening Power=2000)
   &SolveForcesAndVelocity(Clamp Velocity=false, Limit Acceleration=false)
   &DynamicMaterialParameters(
    Index 0 Param 1=1
    Index 0 Param 2=1
    Index 0 Param 3=1
    Index 0 Param 4=&FloatFromCurve(FloatCurve=DataInterface() {propertyValues {Curve {defaultValue=1, keys [<at 0.5 0 RCIM_Cubic>,<at 1 1 RCIM_Cubic>]}}})
    Module.Index0Float0Write=true
    Module.Index0Float1Write=true
    Module.Index0Float2Write=true
    Module.Index0Float3Write=true
   )
  ]
  rendererClasses [&NiagaraMeshRendererProperties, &NiagaraSpriteRendererProperties]
  renderers [
   {rendererClass=&NiagaraMeshRendererProperties, rendererIndex=0}
   {rendererClass=&NiagaraSpriteRendererProperties, rendererIndex=1}
  ]
 }
 LensFlare {
  emitterUpdateScript [
   &EmitterState(
    Inactive Response=::NewEnumerator1
    Life Cycle Mode=::NewEnumerator1
    Loop Behavior=::NewEnumerator1
    Loop Duration Mode=::NewEnumerator1
    UseLoopDelay=false
   )
   &VC_FilteredSocketCount()
   &SpawnBurst_Instantaneous(Spawn Count=2, Use Spawn Probability=false, UseLoopCountLimit=false) {enabled=false}
   &SpawnRate(SpawnRate=300, Use Spawn Probability=false)
   &SpawnRate(SpawnRate=10, Use Spawn Probability=false) {module=SpawnRate001}
  ]
  particleSpawnScript [
   /SetVariables(Particles.VisibilityTag=$Emitter.SpawnGroup) {isSetParameters=true}
   &VC_PaintedVariantsOverride() {enabled=false}
   &InitializeParticle(
    Color Mode=::NewEnumerator0
    Lifetime=&MergeHelper_FloatSwitch(
     0=&RandomRangeFloat(Minimum=1, Maximum=0.5)
     1=1
     2=?
     3=?
     Index=&SpawnGroup()
     Num=::NewEnumerator0
    )
    Position Offset=&AddVectorToPosition(Position=$System.FinalPosition, Vector=<-20 0 0>)
    Sprite Rotation Mode=::NewEnumerator2
    Sprite Size Mode=::NewEnumerator3
    Uniform Sprite Size=&MergeHelper_FloatSwitch(
     0=&RandomRangeFloat(Minimum=150, Maximum=250)
     1=20
     2=?
     3=?
     Index=&SpawnGroup()
     Num=::NewEnumerator0
    )
    UsePositionOffset=true
    Write Material Random=true
   )
   &PartitionParticles(Number Of Partitions=$Emitter.PortCount) {enabled=false}
   &SkeletalMeshLocation(
    Bone / Socket Index=$Particles.PartitionParticles.Partition
    Bone / Socket Sampling Mode=::NewEnumerator5
    MaskSpawnGroup=false
    Mesh Sampling Type=::NewEnumerator6
    Sampled Position Offset=&AddVectorToPosition(Position=<-50 0 0>, Vector=Linked() {linkedVariable {name=User.DTCarOffset, type=NiagaraPosition}})
    Skeletal Mesh=$System.SkeletalMesh
   ) {enabled=false}
   &AddVelocity(Velocity=<-1500 0 0>) {enabled=false}
   &Color(Color=<rgba 1 0.6018 0 1>)
  ]
  particleUpdateScript [
   &ParticleState()
   &ScaleSpriteSize(Uniform Curve Sprite Scale=DataInterface() {propertyValues {Curve {keys [{tangentMode=RCTM_User, time=0, value=2}, <at 0.0927 1>, {tangentMode=RCTM_User, time=1, value=1}]}}})
   &SolveForcesAndVelocity(Clamp Velocity=false, Limit Acceleration=false)
   &ScaleColor(
    Scale RGB=&VectorFromFloat(Value=&FloatFromCurve(FloatCurve=DataInterface() {propertyValues {Curve {keys [<at 0 0 RCIM_Cubic>,<at 0.3 1 RCIM_Cubic>,<at 1 0 RCIM_Cubic>]}}}))
    ScaleA=true
    ScaleRGB=true
   )
  ]
  rendererClasses [&NiagaraSpriteRendererProperties]
  renderers [
   {rendererClass=&NiagaraSpriteRendererProperties}
   {rendererClass=&NiagaraSpriteRendererProperties, rendererIndex=1}
  ]
 }
 NE_VC_Embers001 {
  emitterUpdateScript [
   &EmitterState()
   &VC_FilteredSocketCount()
   &SpawnRate(Spawn Probability=1.5, SpawnRate=200, Use Spawn Probability=false)
  ]
  particleSpawnScript [
   &VC_PaintedVariantsOverride() {enabled=false}
   &InitializeParticle(
    Color=&MakeCustomLinearColorFromBool(
     Bool=$User.bIsPainted
     True Color=$User.PaintedColor
     False Color=&RandomRangeLinearColor(Minimum=<rgba 2 1.3096 0 0.01>, Maximum=<rgba 2 0.35 0.1 0.03>)
    )
    Lifetime Max=0.2
    Lifetime Min=0.15
    Lifetime Mode=::NewEnumerator1
    Position Offset=&AddVectorToPosition(Position=$System.FinalPosition, Vector=<-30 0 0>)
    Position Offset Coordinate Space=::Local
    Sprite Rotation Angle=90
    Sprite Rotation Mode=::NewEnumerator0
    Sprite Size=&RandomRangeVector2D(Minimum=<2 20>, Maximum=<3 50>)
    Sprite Size Mode=::NewEnumerator4
    UsePositionOffset=true
    Write Material Random=true
   )
   &PartitionParticles(Number Of Partitions=$Emitter.PortCount) {enabled=false}
   &SkeletalMeshLocation(
    Bone / Socket Index=$Particles.PartitionParticles.Partition
    Bone / Socket Sampling Mode=::NewEnumerator5
    MaskSpawnGroup=false
    Mesh Sampling Type=::NewEnumerator6
    Sampled Position Offset=&AddVectorToPosition(Position=<-6 0 0>, Vector=$User.DTCarOffset001)
    Skeletal Mesh=$System.SkeletalMesh
   ) {enabled=false}
   &AddVelocity(Velocity=&RandomRangeVector(Minimum=<-288 -550 600>, Maximum=<-1776 550 -100>))
   &ShapeLocation(MaskSpawnGroup=false, Override Seed=false, Sphere Radius=20, UseOwnerScale=true)
  ]
  particleUpdateScript [
   &ParticleState()
   &ScaleSpriteSize(Uniform Curve Sprite Scale=DataInterface() {propertyValues {Curve {keys [{tangentMode=RCTM_User, time=0, value=0}, <at 0.1 1>, {tangentMode=RCTM_User, time=1, value=0}]}}})
   &ScaleVelocity(Velocity Scale=&VectorFromCurve(VectorCurve=DataInterface() {propertyValues {XCurve {keys [<at 0 3>,<at 1 0.5>]}, YCurve {keys [<at 0 3>,<at 1 0.5>]}, ZCurve {keys [<at 0 3>,<at 1 0.5>]}}}))
   &GravityForce(Gravity=&VectorFromCurve(VectorCurve=DataInterface() {propertyValues {XCurve {keys [<at 0 0>,<at 1 0>]}, YCurve {keys [<at 0 0>,<at 1 0>]}, ZCurve {keys [<at 0.5 -29400 RCIM_Cubic>,<at 1 -44100 RCIM_Cubic>]}}}))
   &SolveForcesAndVelocity(Clamp Velocity=false, Limit Acceleration=false)
   &ScaleColor(
    Scale Alpha=&MakeCustomFloatFromBool(Bool=$User.bIsPainted, True Float=&RandomRangeFloat(Minimum=0.03, Maximum=0.06), False Float=1)
    Scale RGB=&VectorFromCurve(
     VectorCurve=DataInterface() {propertyValues {XCurve {keys [<at 0 512>,<at 0.5 32>]}, YCurve {keys [<at 0 512>,<at 0.5 32>]}, ZCurve {keys [<at 0 512>,<at 0.5 32>]}}}
     Scale Curve=&VectorFromFloat(Value=$User.SparksColorScale)
    )
    ScaleA=true
    ScaleRGB=true
   )
   &SubUVAnimation(
    "Sprite Renderer "=DataInterface() {propertyValues {SpriteRenderer=&self:NE_VC_Embers001_0.NiagaraSpriteRendererProperties_0}}
    SubUV Animation Mode:ENiagara_SubUVLookupModeV2=::NewEnumerator2
    UseEndFrame=false
    UseStartFrame=false
   )
  ]
  rendererClasses [&NiagaraSpriteRendererProperties]
  renderers [{rendererClass=&NiagaraSpriteRendererProperties}]
 }
 NE_VC_Embers002 {
  emitterUpdateScript [
   &EmitterState()
   &VC_FilteredSocketCount()
   &SpawnRate(Spawn Probability=1.5, SpawnRate=50, Use Spawn Probability=false)
  ]
  particleSpawnScript [
   &VC_PaintedVariantsOverride() {enabled=false}
   &InitializeParticle(
    Lifetime Max=0.7
    Lifetime Mode=::NewEnumerator1
    Position Offset=&AddVectorToPosition(Position=$System.FinalPosition, Vector=<-30 0 0>)
    Position Offset Coordinate Space=::Local
    Sprite Rotation Mode=::NewEnumerator2
    Sprite Size Mode=::NewEnumerator3
    Uniform Sprite Size=&RandomRangeFloat(Minimum=2, Maximum=4)
    UsePositionOffset=true
    Write Material Random=true
   )
   &PartitionParticles(Number Of Partitions=$Emitter.PortCount) {enabled=false}
   &SkeletalMeshLocation(
    Bone / Socket Index=$Particles.PartitionParticles.Partition
    Bone / Socket Sampling Mode=::NewEnumerator5
    MaskSpawnGroup=false
    Mesh Sampling Type=::NewEnumerator6
    Sampled Position Offset=&AddVectorToPosition(Position=<-6 0 0>, Vector=Linked() {linkedVariable {name=User.DTCarOffset, type=NiagaraPosition}})
    Skeletal Mesh=$System.SkeletalMesh
   ) {enabled=false}
   &AddVelocity(Velocity=&RandomRangeVector(Minimum=<-1200 0 100>, Maximum=<-300 0 -100>)) {enabled=false}
   &AddVelocity(
    Cone Axis=<-1 0 0>
    Module.UseConeFalloff=false
    Module.UseVelDistribution=false
    Override Seed=false
    Velocity Mode=::NewEnumerator2
    Velocity Speed=1000
   ) {module=AddVelocity001}
   &Color(Color=&MakeCustomLinearColorFromBool(
     Bool=$User.bIsPainted
     True Color=$User.PaintedColor
     False Color=&RandomRangeLinearColor(Minimum=<rgba 1 0.6548 0 1>, Maximum=<rgba 1 0.35 0.1 1>)
    ))
  ]
  particleUpdateScript [
   &ParticleState()
   &Collision(
    Age Colliding Particles=true
    Analytical Collision Plane Position 1:Vector3f=<0 0 -40>
    CPU Collision Type=::NewEnumerator3
    Collision Enabled=HlslExpression() {hlslExpression="User.bIsFrontendPreview == true ? true : false"}
    Friction=60
    Friction During a Bounce=0.6
    KillOnCollision=false
    Module.Control Rest State Via Surface Normal=false
    Module.EnableMaxCollisionCount=false
    Module.Randomize Collision Normal Vector=false
    Radius Calculation Type=::NewEnumerator1
    Restitution=0.3
   ) {module=Collision001}
   &Collision(
    Age Colliding Particles=true
    Collision Enabled=HlslExpression() {hlslExpression="User.bIsFrontendPreview == true ? false : true"}
    Control Roll On Collision=true
    Friction=1.5
    Friction During a Bounce:NiagaraFloat=&RandomRangeFloat(Minimum=2, Maximum=1.2)
    KillOnCollision=false
    Module.Control Rest State Via Surface Normal=false
    Module.EnableMaxCollisionCount=false
    Module.Randomize Collision Normal Vector=false
    Radius Scale for Rotation=0
    Restitution=&RandomRangeFloat(Minimum=0, Maximum=0.1)
    ScaleRateOfRotation=true
   )
   &AccelerationForce(Acceleration=&VectorFromCurve(
     VectorCurve=DataInterface() {propertyValues {
       XCurve {keys [<at 0 0>,<at 1 0>]}
       YCurve {keys [<at 0 0>,<at 1 0>]}
       ZCurve {keys [
         {interpMode=RCIM_Cubic, leaveTangentWeight=0.3427, tangentMode=RCTM_User, time=0, value=0}
         {arriveTangent=5.4535, interpMode=RCIM_Cubic, leaveTangent=5.4535, tangentMode=RCTM_User, time=0.5, value=1}
        ]}
      }}
     Scale Curve=<1 1 -8000>
    ))
   &SolveForcesAndVelocity(Clamp Velocity=false, Limit Acceleration=false)
   &ScaleColor(
    Scale RGB=&VectorFromFloat(Value=&FloatFromCurve(FloatCurve=DataInterface() {propertyValues {Curve {keys [<at 0 400 RCIM_Cubic>,<at 0.5 1 RCIM_Cubic>]}}}))
    ScaleA=true
    ScaleRGB=true
   )
  ]
  rendererClasses [&NiagaraSpriteRendererProperties]
  renderers [{rendererClass=&NiagaraSpriteRendererProperties}]
 }
 Trophy001 {
  emitterUpdateScript [
   &EmitterState()
   &VC_FilteredSocketCount()
   &SpawnRate(SpawnRate=3, Use Spawn Probability=true)
  ]
  particleSpawnScript [
   &VC_PaintedVariantsOverride() {enabled=false}
   &InitializeParticle(
    Color Mode=::NewEnumerator0
    Lifetime=1.5
    Mesh Scale Mode=::NewEnumerator3
    Mesh Uniform Scale=1.6
    Position Offset=&AddVectorToPosition(Position=$System.FinalPosition, Vector=<-30 0 0>)
    Position Offset Coordinate Space=::Local
    Sprite Rotation Mode=::NewEnumerator2
    UsePositionOffset=true
    Write Material Random=true
   )
   &PartitionParticles(Number Of Partitions=$Emitter.PortCount) {enabled=false}
   &SkeletalMeshLocation(
    Bone / Socket Index=$Particles.PartitionParticles.Partition
    Bone / Socket Sampling Mode=::NewEnumerator5
    MaskSpawnGroup=false
    Mesh Sampling Type=::NewEnumerator6
    Sampled Position Offset=&AddVectorToPosition(Position=<-20 0 0>, Vector=Linked() {linkedVariable {name=User.DTCarOffset, type=NiagaraPosition}})
    Skeletal Mesh=$System.SkeletalMesh
   ) {enabled=false}
   &AddVelocity(Velocity=&RandomRangeVector(Minimum=<-600 -120 300>, Maximum=<-1000 120 50>))
   &Color(Color=<rgba 0.4479 0.4479 0.4479 1>) {enabled=false}
   &InitialMeshOrientation(
    Mesh Orientation Mode=::NewEnumerator7
    Rotation=&RandomRangeVector(Minimum=<-1 -1 1>, Maximum=<1 1 1>)
    Use Rotation Vector=true
   )
  ]
  particleUpdateScript [
   &ParticleState()
   &GravityForce(Gravity=&RandomRangeVector(Minimum=<0 0 -2000>, Maximum=<0 0 -1200>))
   &InheritSourceMovement(
    Applied Position Scale=&VectorFromCurve(VectorCurve=DataInterface() {propertyValues {XCurve {keys [<at 0.1 0.7>,<at 0.5 0>]}, YCurve {keys [<at 0.1 0.7>,<at 0.5 0>]}, ZCurve {keys [<at 0.1 0.7>,<at 0.5 0>]}}})
    ApplyForce=false
    ApplyPosition=true
    ApplyPositionFalloff=false
   )
   &Collision(
    Advanced Aging Rate=5
    Age Colliding Particles=true
    Collision Enabled=HlslExpression() {hlslExpression="User.bIsFrontendPreview == true ? false : true"}
    Control Roll On Collision=true
    Friction=1.5
    Friction During a Bounce:NiagaraFloat=&RandomRangeFloat(Minimum=2, Maximum=1.2)
    KillOnCollision=false
    Module.Control Rest State Via Surface Normal=false
    Module.EnableMaxCollisionCount=false
    Module.Randomize Collision Normal Vector=false
    Radius Scale for Rotation=0
    Restitution=&RandomRangeFloat(Minimum=0, Maximum=0.1)
    ScaleRateOfRotation=true
   )
   &Collision(
    Advanced Aging Rate=0
    Age Colliding Particles=true
    Analytical Collision Plane Position 1:Vector3f=<0 0 55>
    CPU Collision Type=::NewEnumerator3
    CPU Friction Blending:ENiagaraFrictionMergeType=::NewEnumerator1
    Collision Enabled=HlslExpression() {hlslExpression="User.bIsFrontendPreview == true ? true : false"}
    Control Roll On Collision=true
    Friction=1
    Friction During a Bounce=1
    KillOnCollision=false
    Max Number Of Collisions=2
    Maximum Penetration Correction Distance=100
    Module.Control Rest State Via Surface Normal=false
    Module.EnableMaxCollisionCount=true
    Module.Randomize Collision Normal Vector=true
    Radius Calculation Type=::NewEnumerator1
    Radius Scale for Rotation=200
    Restitution=0.3
    Restitution Coefficient Blending:ENiagaraRestitutionMergeType=::NewEnumerator2
    ScaleRateOfRotation=true
   ) {module=Collision001}
   &AlignParticlesWithCollisionPlane(Force Alignment=true, Rotational Dampening Power=200)
   &SolveForcesAndVelocity(Clamp Velocity=false, Limit Acceleration=false)
   &DynamicMaterialParameters(
    Index 0 Param 2=1
    Index 0 Param 4=&FloatFromCurve(FloatCurve=DataInterface() {propertyValues {Curve {
        defaultValue=1
        keys [<at 0.5 0 RCIM_Cubic>, {arriveTangent=5.8244, interpMode=RCIM_Cubic, leaveTangent=5.8244, tangentMode=RCTM_User, time=1, value=1}]
       }}})
    Module.Index0Float0Write=true
    Module.Index0Float1Write=true
    Module.Index0Float2Write=true
    Module.Index0Float3Write=true
   )
  ]
  rendererClasses [&NiagaraMeshRendererProperties, &NiagaraSpriteRendererProperties]
  renderers [
   {rendererClass=&NiagaraMeshRendererProperties, rendererIndex=0}
   {rendererClass=&NiagaraSpriteRendererProperties, rendererIndex=1}
  ]
 }
}
summary {
 systemName=NS_GoldBars
 emitters [
  {emitterName=LensFlare, rendererClasses [&NiagaraSpriteRendererProperties]}
  {emitterName=Goldbar, rendererClasses [&NiagaraMeshRendererProperties, &NiagaraSpriteRendererProperties]}
  {emitterName=NE_VC_Embers002, rendererClasses [&NiagaraSpriteRendererProperties]}
  {emitterName=Trophy001, rendererClasses [&NiagaraMeshRendererProperties, &NiagaraSpriteRendererProperties]}
  {emitterName=NE_VC_Embers001, rendererClasses [&NiagaraSpriteRendererProperties]}
 ]
 userVariables [
  User.AlphaScale {type=NiagaraFloat, defaultValue=1, description=""}
  User.bIsFrontendPreview {type=NiagaraBool}
  User.bIsPainted {type=NiagaraBool}
  User.ColorVariantID {type=NiagaraInt32}
  User.DTCarOffset {type=NiagaraPosition, description="", defaultValue {}}
  User.DTCarOffset001 {type=Vector3f, description="", defaultValue {struct=Vector3f, value=<0 0 0>}}
  User.PaintedColor {type=LinearColor, description="", defaultValue {struct=LinearColor, value=<rgba 1 1 1 1>}}
  User.PaintedScalar {type=NiagaraFloat, defaultValue=1, description=""}
  User.PortCount {type=NiagaraInt32, defaultValue=1}
  User.ScaleColorFactor {type=NiagaraFloat, defaultValue=1, description=""}
  User.SingleBoostPos {type=NiagaraPosition, description="", defaultValue {}}
  User.SingleBoostPos {type=Vector3f, description="", defaultValue {struct=Vector3f, value=<0 0 0>}}
  User.SparksColorScale {type=NiagaraFloat, defaultValue=2, description=""}
 ]
}
